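# Level Sequence for the face animation of the NotchGnarl male skin's lobby idle, built on one FortPlayerMannequin binding with a fixed transform.
# A skeletal animation track spans frames 0 to 324 at a 30 fps display rate.
# A Fortnite_Legacy_Face_CtrlRig track holds constant values for the master control's placement, head-follow and a disabled Flipbook switch, with no varying facial channels shown.
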LevelSequence /BRCosmetics/Animation/Game/MainPlayer/Menu/BR/CustomSkins/_s40/NotchGnarl_M/Anim_NotchGnarl_Lobby_Idle/FaceAnim_NotchGnarl_Lobby_Idle
bindings [FortPlayerMannequin {
  id=#1
  possessed_class=FortPlayerMannequin
  tracks [
   {type=MovieSceneSpawnTrack, sections [{type=MovieSceneSpawnSection, channels [None {type=MovieSceneScriptingBoolChannel, default=true}]}]}
   {
    type=MovieScene3DTransformTrack
    sections [{
      type=MovieScene3DTransformSection
      channels [
       Location.X {type=MovieSceneScriptingDoubleChannel, default=0}
       Location.Y {type=MovieSceneScriptingDoubleChannel, default=0}
       Location.Z {type=MovieSceneScriptingDoubleChannel, default=0}
       Rotation.X {type=MovieSceneScriptingDoubleChannel, default=0}
       Rotation.Y {type=MovieSceneScriptingDoubleChannel, default=0}
       Rotation.Z {type=MovieSceneScriptingDoubleChannel, default=0}
       Scale.X {type=MovieSceneScriptingDoubleChannel, default=1}
       Scale.Y {type=MovieSceneScriptingDoubleChannel, default=1}
       Scale.Z {type=MovieSceneScriptingDoubleChannel, default=1}
      ]
     }]
   }
   {
    type=MovieSceneControlRigParameterTrack
    display_name=Fortnite_Legacy_Face_CtrlRig
    sections [{
      type=MovieSceneControlRigParameterSection
      channels [
       Flipbook {type=MovieSceneScriptingBoolChannel, default=false, const [false,1]}
       Face_Ctrl_Master.Location.X {type=MovieSceneScriptingFloatChannel, default=30.0712, const [30.0712,1]}
       Face_Ctrl_Master.Location.Y {type=MovieSceneScriptingFloatChannel, default=0.0334, const [0.0334,1]}
       Face_Ctrl_Master.Location.Z {type=MovieSceneScriptingFloatChannel, default=4.0374, const [4.0374,1]}
       Face_Ctrl_Master.Rotation.X {type=MovieSceneScriptingFloatChannel, default=0, const [0,1]}
       Face_Ctrl_Master.Rotation.Y {type=MovieSceneScriptingFloatChannel, default=0, const [0,1]}
       Face_Ctrl_Master.Rotation.Z {type=MovieSceneScriptingFloatChannel, default=0, const [0,1]}
       Face_Ctrl_Master.Scale.X {type=MovieSceneScriptingFloatChannel, default=2.3101, const [2.3101,1]}
       Face_Ctrl_Master.Scale.Y {type=MovieSceneScriptingFloatChannel, default=2.3101, const [2.3101,1]}
       Face_Ctrl_Master.Scale.Z {type=MovieSceneScriptingFloatChannel, default=2.3101, const [2.3101,1]}
       gui_follow_head {type=MovieSceneScriptingFloatChannel, default=1, const [1,1]}
      ]
     }]
   }
   {type=MovieSceneSkeletalAnimationTrack, sections [{type=MovieSceneSkeletalAnimationSection, end_frame=324, start_frame=0}]}
  ]
 }]
display_rate {denominator=1, numerator=30}
tick_resolution {denominator=1, numerator=24000}
tracks []
